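# Lit material for the race track start light's score display, showing position and minute/second timer digits as a grid of tinted bulbs.
# Scalar parameters shift a number texture horizontally in tenths to pick each digit, with vertex color RGBA masking the four timer digits.
# The position digit uses an inverse vertex-color mask that sums red twice and omits green.
# Digits and a UV2 gradient mask multiply a squared, tiled bulb texture, Color, Flash and Brightness.
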Material /Game/Environments/Sets/RaceTrack/Materials/M_Racetrack_StartLight_Score
graph {
 properties {pixelDepthOffsetMode=PDOM_Legacy, usedShadingModels=Default Lit}
 nodes [
  n1 = TextureSample(UVs=TextureCoordinate() {coordinateIndex=1, uTiling=30, vTiling=9}) {samplerType=SAMPLERTYPE_LinearColor, texture=/Game/Environments/Sets/RaceTrack/Textures/T_Racetrack_StartLight_NumberBulb}
  LinearInterpolate(A=Constant(), B=Constant() {r=0.2}, Alpha=@n1.R)
  LinearInterpolate(A=Constant() {r=0.05}, B=Constant() {r=0.1}, Alpha=@n1.R)
  n2 = Power(Base=@n1.RGB, Exp=Constant() {r=2})
  n3 = ComponentMask(None=TextureCoordinate() {coordinateIndex=2}) {g=true}
  n4 = Add(A=Constant() {r=-3}, B=Multiply(A=Constant() {r=4}, B=ConstantBiasScale(None=@n3)))
  n5 = VertexColor()
  n6 = Add(A=Add(A=@n5.R, B=@n5.R), B=@n5.B)
  n7 = MaterialFunctionCall(X=Multiply(A=ScalarParameter() {parameterName=Position}, B=Constant() {r=0.1}), Y=Constant()) {materialFunction=&MakeFloat2}
  n8 = TextureCoordinate()
  n9 = Multiply(
   A=OneMinus(None=Add(A=@n6, B=@n5.A))
   B=TextureSample(UVs=Add(A=@n7.Result, B=@n8)) {texture=/Game/Environments/Sets/RaceTrack/Textures/T_Racetrack_StartLight_SimpleNumbers}.R
  )
  n10 = VertexColor()
  n11 = Multiply(A=ScalarParameter() {parameterName=TenMinute}, B=Constant() {r=0.1})
  n12 = Constant()
  n13 = TextureSample(UVs=Add(
    A=MaterialFunctionCall(X=@n11, Y=@n12) {materialFunction=&MakeFloat2}.Result
    B=TextureCoordinate()
   )) {texture=&T_Racetrack_StartLight_SimpleNumbers}
  n14 = Multiply(A=ScalarParameter() {parameterName=OneMinute}, B=Constant() {r=0.1})
  n15 = Constant()
  n16 = TextureSample(UVs=Add(
    A=MaterialFunctionCall(X=@n14, Y=@n15) {materialFunction=&MakeFloat2}.Result
    B=TextureCoordinate()
   )) {texture=&T_Racetrack_StartLight_SimpleNumbers}
  n17 = Multiply(A=ScalarParameter() {parameterName=TenSecond}, B=Constant() {r=0.1})
  n18 = Constant()
  n19 = TextureSample(UVs=Add(
    A=MaterialFunctionCall(X=@n17, Y=@n18) {materialFunction=&MakeFloat2}.Result
    B=TextureCoordinate()
   )) {texture=&T_Racetrack_StartLight_SimpleNumbers}
  n20 = Multiply(A=ScalarParameter() {parameterName=OneSecond}, B=Constant() {r=0.1})
  n21 = Constant()
  n22 = TextureSample(UVs=Add(
    A=MaterialFunctionCall(X=@n20, Y=@n21) {materialFunction=&MakeFloat2}.Result
    B=TextureCoordinate()
   )) {texture=&T_Racetrack_StartLight_SimpleNumbers}
  n23 = Add(A=Add(A=Multiply(A=@n10.R, B=@n13.R), B=Multiply(A=@n10.G, B=@n16.R)), B=Add(A=Multiply(A=@n10.B, B=@n19.R), B=Multiply(A=@n10.A, B=@n22.R)))
  n24 = Multiply(A=Add(A=Clamp(None=@n4), B=Add(A=@n9, B=@n23)), B=VectorParameter() {parameterName=Color, defaultValue=<rgba 3 2.6607 1.2 1>}.RGB)
  Multiply(
   A=Multiply(A=Multiply(A=@n2, B=@n24), B=ScalarParameter() {defaultValue=1, parameterName=Flash})
   B=ScalarParameter() {defaultValue=1, parameterName=Brightness}
  )
 ]
}
