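# Material for the racetrack start light timer board, showing numbers, lap time, animated arrows or a checkered flag as a lit-bulb display.
# A number atlas is offset by digit parameters, with vertex color channels masking the four lap-time digits.
# Scalar parameters blend between modes, including stepped-panner arrows height-blended with bulbs and a time-stepped, sine-pulsed checker.
# An optional inverted flash precedes multiplication by the tiled bulb texture, color and brightness.
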Material /Game/Environments/Sets/RaceTrack/Materials/M_Racetrack_StartLight_Timer
graph {
 properties {
  bUsedWithNanite=true
  pixelDepthOffsetMode=PDOM_Legacy
  previewMesh=/Game/Environments/Sets/RaceTrack/Meshes/S_Racetrack_StartLight_Timer
  usedShadingModels=Default Lit
 }
 nodes [
  n1 = ScalarParameter() {defaultValue=1, parameterName=Slider}
  n2 = ScalarParameter() {parameterName=LapTime}
  n3 = Multiply(A=ScalarParameter() {defaultValue=1, parameterName=TenMinute}, B=Constant() {r=0.1})
  n4 = Constant() {r=-0.033}
  n5 = Add(
   A=MaterialFunctionCall(X=Add(A=@n3, B=@n4), Y=Constant()) {materialFunction=&MakeFloat2}.Result
   B=TextureCoordinate() {uTiling=5}
  )
  n6 = Multiply(A=ScalarParameter() {defaultValue=1, parameterName=DigitA}, B=Constant() {r=0.1})
  n7 = Constant()
  n8 = TextureCoordinate() {uTiling=5}
  n9 = ScalarParameter() {defaultValue=0.5, parameterName=DigitLoc}
  n10 = LinearInterpolate(A=Constant() {r=-0.5}, B=Constant() {r=0.1}, Alpha=@n9)
  n11 = LinearInterpolate(A=Constant() {r=-0.5}, B=Constant() {r=0.2}, Alpha=@n9)
  n12 = ScalarParameter() {defaultValue=1, parameterName=Use2Digits}
  n13 = Constant() {r=30}
  n14 = Multiply(A=Frac(None=LinearInterpolate(A=@n10, B=@n11, Alpha=@n12)), B=@n13)
  n15 = MaterialFunctionCall(X=Divide(A=Ceil(None=@n14), B=@n13), Y=Constant()) {materialFunction=&MakeFloat2}
  n16 = Add(
   A=Add(A=MaterialFunctionCall(X=@n6, Y=@n7) {materialFunction=&MakeFloat2}.Result, B=@n8)
   B=@n15.Result
  )
  n17 = TextureSample(UVs=Add(A=Multiply(A=@n2, B=@n5), B=Multiply(A=@n1, B=@n16))) {texture=/Game/Environments/Sets/RaceTrack/Textures/T_Racetrack_StartLight_SimpleNumbers}
  n18 = Multiply(A=ScalarParameter() {defaultValue=2, parameterName=OneMinute}, B=Constant() {r=0.1})
  n19 = Constant() {r=-0.133}
  n20 = Add(
   A=MaterialFunctionCall(X=Add(A=@n18, B=@n19), Y=Constant()) {materialFunction=&MakeFloat2}.Result
   B=TextureCoordinate() {uTiling=5}
  )
  n21 = Add(A=@n8, B=@n15.Result)
  n22 = TextureSample(UVs=Add(A=Multiply(A=@n2, B=@n20), B=Multiply(A=@n1, B=@n21))) {texture=&T_Racetrack_StartLight_SimpleNumbers}
  n23 = Multiply(A=ScalarParameter() {defaultValue=3, parameterName=TenSecond}, B=Constant() {r=0.1})
  n24 = Constant() {r=-0.25}
  n25 = Add(
   A=MaterialFunctionCall(X=Add(A=@n23, B=@n24), Y=Constant()) {materialFunction=&MakeFloat2}.Result
   B=TextureCoordinate() {uTiling=5}
  )
  n26 = LinearInterpolate(
   A=Constant() {r=-0.1}
   B=Constant() {r=0.9}
   Alpha=Multiply(A=ScalarParameter() {defaultValue=5, parameterName=DigitB}, B=Constant() {r=0.1})
  )
  n27 = Constant()
  n28 = TextureCoordinate() {uTiling=5}
  n29 = Add(
   A=Add(A=MaterialFunctionCall(X=@n26, Y=@n27) {materialFunction=&MakeFloat2}.Result, B=@n28)
   B=@n15.Result
  )
  n30 = TextureSample(UVs=Add(A=Multiply(A=@n2, B=@n25), B=Multiply(A=@n1, B=@n29))) {texture=&T_Racetrack_StartLight_SimpleNumbers}
  n31 = Multiply(A=ScalarParameter() {defaultValue=4, parameterName=OneSecond}, B=Constant() {r=0.1})
  n32 = Constant() {r=-0.35}
  n33 = Add(
   A=MaterialFunctionCall(X=Add(A=@n31, B=@n32), Y=Constant()) {materialFunction=&MakeFloat2}.Result
   B=TextureCoordinate() {uTiling=5}
  )
  n34 = Add(A=@n15.Result, B=@n28)
  n35 = TextureSample(UVs=Add(A=Multiply(A=@n2, B=@n33), B=Multiply(A=@n1, B=@n34))) {texture=&T_Racetrack_StartLight_SimpleNumbers}
  n36 = Add(A=Multiply(A=@n17.R, B=@n22.G), B=Multiply(A=@n12, B=Multiply(A=@n30.R, B=@n35.B)))
  Add(B=Multiply(A=@n1, B=@n36))
  n37 = TextureSample(UVs=TextureCoordinate() {uTiling=300, vTiling=9}) {samplerType=SAMPLERTYPE_LinearColor, texture=/Game/Environments/Sets/RaceTrack/Textures/T_Racetrack_StartLight_NumberBulb}
  n38 = Power(Base=@n37.RGB, Exp=Constant() {r=2})
  n39 = Constant() {r=1}
  n40 = ScalarParameter() {defaultValue=8, parameterName=rot_4}
  n41 = ScalarParameter() {defaultValue=16, parameterName=rot_5}
  n42 = Add(
   A=MaterialFunctionCall!(
    TimeSpeed=ScalarParameter() {defaultValue=0.5, parameterName=rot_1}
    Jump Amount Per Step=MaterialFunctionCall(X=@n39, Y=@n40) {materialFunction=&MakeFloat2}.Result
    Number of Steps Per Unit of Time=MaterialFunctionCall(X=@n39, Y=@n41) {materialFunction=&MakeFloat2}.Result
    Time Offset
   ) {materialFunction=/Engine/Functions/Engine_MaterialFunctions02/SteppingPannerTime}.Result
   B=TextureCoordinate() {coordinateIndex=1, vTiling=0.5}
  )
  n43 = Constant() {r=0.25}
  n44 = MaterialFunctionCall(
   Transition Phase=TextureSample(UVs=MaterialFunctionCall(UVs=@n42, "Rotation Angle (0-1)"=@n43) {materialFunction=&CustomRotator}.Rotated Values) {texture=/Game/UI/Foundation/Textures/Pattern/Pattern-HeavyArrows}.R
   Height Texture=@n38
  ) {materialFunction=&HeightLerp}
  n45 = ScalarParameter() {parameterName=Arrows}
  n46 = TextureCoordinate() {uTiling=60, vTiling=1.15}
  n47 = Multiply(A=Time(), B=ScalarParameter() {defaultValue=2, parameterName=FlagSpeed})
  n48 = AppendVector(A=Floor(None=@n47), B=Constant())
  n49 = MaterialFunctionCall!(UVs=Add(A=Constant2Vector() {g=0.233}, B=Add(A=@n46, B=@n48)), Tiling=Constant2Vector() {g=2, r=1}) {materialFunction=/Engine/ArtTools/RenderToTexture/MaterialFunctions/CheckerPattern}
  n50 = Add(A=@n47, B=Constant() {r=-0.2})
  n51 = Add(A=ScalarParameter() {defaultValue=0.1, parameterName=rot_2}, B=ConstantBiasScale(None=Sine(None=@n50)))
  n52 = VertexColor()
  n53 = Multiply(A=@n52.R, B=@n17.R)
  n54 = Multiply(A=@n52.G, B=@n22.R)
  n55 = Multiply(A=@n52.B, B=@n30.R)
  n56 = Multiply(A=@n52.A, B=@n35.R)
  n57 = Constant() {r=-3}
  n58 = TextureCoordinate() {coordinateIndex=2}
  n59 = Multiply(A=Constant() {r=4}, B=ConstantBiasScale(None=ComponentMask(None=@n58) {g=true}))
  n60 = Add(A=Add(A=Add(A=@n53, B=@n54), B=Add(A=@n55, B=@n56)), B=Clamp(None=Add(A=@n57, B=@n59)))
  n61 = LinearInterpolate(
   A=LinearInterpolate(
    A=LinearInterpolate(A=@n36, B=@n44.Lerp Alpha No Contrast, Alpha=@n45)
    B=Multiply(A=@n49.Result, B=@n51)
    Alpha=ScalarParameter() {parameterName=Flag}
   )
   B=@n60
   Alpha=@n2
  )
  n62 = LinearInterpolate(A=@n61, B=Multiply(A=OneMinus(None=@n61), B=ScalarParameter() {parameterName=Flash}), Alpha=ScalarParameter() {parameterName=FlashControl})
  n63 = VectorParameter() {parameterName=Color, defaultValue=<rgba 3 2.6607 1.2 1>}
  Multiply(A=Multiply(A=@n38, B=Multiply(A=@n62, B=@n63.RGB)), B=ScalarParameter() {parameterName=Brightness})
  Multiply(A=@n60, B=@n2)
  Reroute(None=LinearInterpolate(A=ScalarParameter() {defaultValue=0.05, parameterName=DiffA}, B=ScalarParameter() {defaultValue=0.1, parameterName=DiffB}, Alpha=@n37.R))
  Reroute(None=LinearInterpolate(A=Constant(), B=Constant() {r=0.2}, Alpha=@n37.R))
  TextureCoordinate() {coordinateIndex=1, uTiling=30, vTiling=9}
 ]
}
