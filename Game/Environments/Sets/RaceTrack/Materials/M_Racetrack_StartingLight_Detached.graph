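# Unlit additive material for the racetrack starting light, glowing amber, green, red or warm white per light group.
# An exponential radial gradient on the UVs blends two HDR colours per group, masked by vertex colour channels.
# Scalars L1–L5 scale the groups and default to zero, with L4-FalseStart added to red L4.
# The small-light mask inverts the summed channels, but counts R twice and omits G.
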Material /Game/Environments/Sets/RaceTrack/Materials/M_Racetrack_StartingLight_Detached
graph {
 properties {bTangentSpaceNormal=false, blendMode=BLEND_Additive, pixelDepthOffsetMode=PDOM_Legacy, shadingModel=MSM_Unlit, usedShadingModels=Unlit}
 nodes [
  n1 = MaterialFunctionCall(
   UVs=TextureCoordinate()
   Radius=ScalarParameter() {defaultValue=0.3, parameterName=GradRad}
   Density=ScalarParameter() {defaultValue=1, parameterName=GradDens}
  ) {materialFunction=&RadialGradientExponential}
  n2 = LinearInterpolate(
   A=VectorParameter() {parameterName=Light2_1, defaultValue=<rgba 2 0.1677 0 1>}.RGB
   B=VectorParameter() {parameterName=Light2, defaultValue=<rgba 50 37.5 0 1>}.RGB
   Alpha=@n1.RadialGradientExponential
  )
  n3 = VertexColor()
  n4 = Add(
   A=Multiply(A=ScalarParameter() {parameterName=L1}, B=Multiply(A=@n2, B=@n3.R))
   B=Multiply(A=ScalarParameter() {parameterName=L2}, B=Multiply(A=@n2, B=@n3.G))
  )
  n5 = VectorParameter() {parameterName=Light3_1, defaultValue=<rgba 0 1.5 0 1>}
  n6 = VectorParameter() {parameterName=Light3, defaultValue=<rgba 0 60 28.9503 1>}
  n7 = Multiply(A=ScalarParameter() {parameterName=L3}, B=Multiply(A=LinearInterpolate(A=@n5.RGB, B=@n6.RGB, Alpha=@n1.RadialGradientExponential), B=@n3.B))
  n8 = Add(A=ScalarParameter() {parameterName=L4}, B=ScalarParameter() {parameterName=L4-FalseStart})
  n9 = Multiply(
   A=LinearInterpolate(
    A=VectorParameter() {parameterName=Light4_1, defaultValue=<rgba 3 0 0 1>}.RGB
    B=VectorParameter() {parameterName=Light4, defaultValue=<rgba 40 17.5584 0 1>}.RGB
    Alpha=@n1.RadialGradientExponential
   )
   B=@n3.A
  )
  n10 = LinearInterpolate(
   A=VectorParameter() {parameterName=SmallLights_1, defaultValue=<rgba 0.5 0.2973 0.05 1>}.RGB
   B=VectorParameter() {parameterName=SmallLights, defaultValue=<rgba 10 10 7.5 1>}.RGB
   Alpha=@n1.RadialGradientExponential
  )
  n11 = Add(A=Add(A=@n3.R, B=@n3.R), B=@n3.B)
  n12 = OneMinus(None=Clamp(None=Add(A=@n11, B=@n3.A)))
  Add(A=Add(A=Add(A=@n4, B=@n7), B=Multiply(A=@n8, B=@n9)), B=Multiply(A=ScalarParameter() {parameterName=L5}, B=Multiply(A=@n10, B=@n12)))
  ScalarParameter() {defaultValue=1, parameterName=Opacity}
 ]
}
